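# Sound cue for the front-end UI, named as the big "Ranked Off" sound, likely heard when ranked mode is toggled off.
# A single wave player feeds the ranked select toggle-off wave straight to the output, with lowered volume and slightly lowered pitch on the Fort_UI sound class.
# A concurrency override allows only a couple of simultaneous instances, stopping the oldest, with a short retrigger time.
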SoundCue /Game/Sounds/Fort_UI_Sounds/FrontEnd/New_FrontEnd/Ranked/Fortnite_UI_Ranked_Off_Big_Cue
root=WavePlayer_1
nodes [{id=WavePlayer_1, properties {soundWaveAssetPtr=/Game/Sounds/Fort_UI_Sounds/FrontEnd/New_FrontEnd/Ranked/wav/UI_Ranked_Select_Toggle_Off_New}}]
properties {
 bOverrideConcurrency=true
 maxDistance=2097152
 pitchMultiplier=0.9
 soundClassObject=&Fort_UI
 volumeMultiplier=0.5
 concurrencyOverrides {maxCount=2, resolutionRule=StopOldest, retriggerTime=0.1}
}
