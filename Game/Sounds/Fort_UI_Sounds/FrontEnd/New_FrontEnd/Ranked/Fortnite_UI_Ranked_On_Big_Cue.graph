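# Sound cue for a front-end Ranked UI "toggle on" sound.
# A single wave player feeds the Ranked select toggle-on wave straight to the output in the Fort_UI sound class at slightly reduced volume.
# A concurrency override caps simultaneous plays at a small count, stopping the oldest and applying a short retrigger time to tame rapid toggling.
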SoundCue /Game/Sounds/Fort_UI_Sounds/FrontEnd/New_FrontEnd/Ranked/Fortnite_UI_Ranked_On_Big_Cue
root=WavePlayer_0
nodes [{id=WavePlayer_0, properties {soundWaveAssetPtr=/Game/Sounds/Fort_UI_Sounds/FrontEnd/New_FrontEnd/Ranked/wav/UI_Ranked_Select_Toggle_On_New}}]
properties {
 bOverrideConcurrency=true
 maxDistance=2097152
 soundClassObject=&Fort_UI
 volumeMultiplier=0.8
 concurrencyOverrides {maxCount=2, resolutionRule=StopOldest, retriggerTime=0.1}
}
