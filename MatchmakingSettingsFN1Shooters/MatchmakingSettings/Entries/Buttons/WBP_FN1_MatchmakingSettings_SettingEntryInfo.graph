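# Label widget for a setting entry button in the FN1 Shooters matchmaking settings, showing an uppercase setting name and optional secondary text.
# AddImageBlock lazily creates an icon block in the grid column left of the text, then applies padding and size.
# SetFocused, SetSelected and SetEnabled play fast forward or reverse animations on text color or font material.
# With bBackgroundOnlyForNamedSlot set, SetEnabled changes text opacity instead of animating.
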WidgetBlueprint /MatchmakingSettingsFN1Shooters/MatchmakingSettings/Entries/Buttons/WBP_FN1_MatchmakingSettings_SettingEntryInfo
animations [
 Anim_OnFocus {
  binding_count=2
  length_seconds=1
  object=&self:Anim_OnFocus
  sequence {bindings [
    Text {
     id=#1
     possessed_class=CommonTextBlock
     tracks [{
       type=MovieSceneColorTrack
       display_name=Color and Opacity
       sections [{
         type=MovieSceneColorSection
         end_frame=20
         start_frame=0
         channels [
          Color.R {type=MovieSceneScriptingFloatChannel, keys [[0,1],[16,1],[20,0.005]]}
          Color.G {type=MovieSceneScriptingFloatChannel, keys [[0,1],[16,1],[20,0.004]]}
          Color.B {type=MovieSceneScriptingFloatChannel, keys [[0,1],[16,1],[20,0.012]]}
          Color.A {type=MovieSceneScriptingFloatChannel, const [1,2]}
         ]
        }]
      }]
    }
    SecondText {
     id=#2
     possessed_class=CommonTextBlock
     tracks [{
       type=MovieSceneColorTrack
       display_name=Color and Opacity
       sections [{
         type=MovieSceneColorSection
         end_frame=20
         start_frame=0
         channels [
          Color.R {type=MovieSceneScriptingFloatChannel, keys [[0,0.6939],[20,0.005]]}
          Color.G {type=MovieSceneScriptingFloatChannel, keys [[0,0.6939],[20,0.004]]}
          Color.B {type=MovieSceneScriptingFloatChannel, keys [[0,0.6939],[20,0.012]]}
          Color.A {type=MovieSceneScriptingFloatChannel, keys [[0,0.749],[20,1]]}
         ]
        }]
      }]
    }
   ]}
 }
 Anim_OnDisabled {
  binding_count=2
  length_seconds=1
  object=&self:Anim_OnDisabled
  sequence {bindings [
    Text {
     id=#3
     possessed_class=CommonTextBlock
     tracks [{
       type=MovieSceneColorTrack
       display_name=Color and Opacity
       sections [{
         type=MovieSceneColorSection
         end_frame=20
         start_frame=0
         channels [
          Color.R {type=MovieSceneScriptingFloatChannel, keys [[0,1],[20,0.074]]}
          Color.G {type=MovieSceneScriptingFloatChannel, keys [[0,1],[20,0.04]]}
          Color.B {type=MovieSceneScriptingFloatChannel, keys [[0,1],[20,0.381]]}
          Color.A {type=MovieSceneScriptingFloatChannel, const [1,1]}
         ]
        }]
      }]
    }
    SecondText {
     id=#4
     possessed_class=CommonTextBlock
     tracks [{
       type=MovieSceneColorTrack
       display_name=Color and Opacity
       sections [{
         type=MovieSceneColorSection
         end_frame=20
         start_frame=0
         channels [
          Color.R {type=MovieSceneScriptingFloatChannel, keys [[0,0.6939],[20,0.074]]}
          Color.G {type=MovieSceneScriptingFloatChannel, keys [[0,0.6939],[20,0.04]]}
          Color.B {type=MovieSceneScriptingFloatChannel, keys [[0,0.6939],[20,0.381]]}
          Color.A {type=MovieSceneScriptingFloatChannel, keys [[0,0.749],[20,1]]}
         ]
        }]
      }]
    }
   ]}
 }
 Anim_Selected {
  binding_count=1
  length_seconds=1
  object=&self:Anim_Selected
  sequence {bindings [Text {
     id=#5
     possessed_class=CommonTextBlock
     tracks [{
       type=MovieSceneWidgetMaterialTrack
       display_name=Font.FontMaterial
       sections [{
         type=MovieSceneParameterSection
         end_frame=20
         start_frame=0
         channels [Trans_SelectedIdle {type=MovieSceneScriptingFloatChannel, keys [[0,0],[20,1]]}]
        }]
      }]
    }]}
 }
]
graphs [
 AddImageBlock {nodes [
   AddImageBlock() {is=title K2Node_FunctionEntry}
   > n1 = Get!(execute:Exec) -> (then:Exec, else:Exec, Block_Icon:WBP Matchmaking Block Image Object Reference)
   n2 = Create Widget!(
    execute:Exec=@n1.else
    Class:User Widget Class Reference=/MatchmakingSettingsFN1Shooters/MatchmakingSettings/WBP_MatchmakingBlock_Image.WBP_MatchmakingBlock_Image_C
    OwningPlayer:Player Controller Object Reference=GetOwningPlayer()
   ) -> (then:Exec, ReturnValue:WBP Matchmaking Block Image Object Reference)
   > n3 = Set Block_Icon(Block_Icon:WBP Matchmaking Block Image Object Reference=@n2) -> (Output_Get:WBP Matchmaking Block Image Object Reference)
   n4 = AddChildToGrid!(
    execute=[@n1.then, @n3.then]
    self:Grid Panel Object Reference=Get Main() -> (Main:Grid Panel Object Reference)
    Content:Widget Object Reference=@n3
    InRow:int=0
    InColumn:int=0
   ) -> (then, ReturnValue:Grid Slot Object Reference)
   > SetPadding!(execute, self:Grid Slot Object Reference=@n4, InPadding:Margin Structure=Get ImagePadding() -> (ImagePadding:Margin Structure)) -> (then)
   n5 = Get Block_Icon() -> (Block_Icon:WBP Matchmaking Block Image Object Reference)
   > SetSize!(
    execute
    self:WBP Matchmaking Block Image Object Reference=@n5
    Desired Size:Vector 2D Structure=Get ImageSize() -> (ImageSize:Vector 2D Structure)
   ) -> (then)
   > Return Node!(execute, Block_Icon:WBP Matchmaking Block Image Object Reference=@n5)
  ]}
 EventGraph {links [], nodes []}
 SetEnabled {nodes [
   n1 = SetEnabled!() -> (then, Enabled:bool=~) {is=title K2Node_FunctionEntry}
   > n2 = Branch(Condition=@n1)
   n3 = Get bBackgroundOnlyForNamedSlot() -> (bBackgroundOnlyForNamedSlot:bool)
   > n4 = Branch(Condition=@n3)
   SetRenderOpacity(
    execute=[@n4.then, @n5.then]
    self=[Get SecondText() -> (SecondText:Common Text Object Reference), Get Text!() -> (Text:Common Text Object Reference)]
    InOpacity=Select(Option 0:float=0.5, Option 1:float=1, Index=Get Enabled() -> (Enabled:bool)) -> (ReturnValue:float)
   )
   PlayAnimationReverse(execute=@n4.else, InAnimation=Get Anim_OnDisabled() -> (Anim_OnDisabled:Widget Animation Object Reference), PlaybackSpeed=10)
   n5 = Branch(execute=@n2.else, Condition=@n3)
   PlayAnimationForward(execute=@n5.else, InAnimation=Get Anim_OnDisabled(), PlaybackSpeed=10)
  ]}
 SetFocused {nodes [
   n1 = SetFocused!() -> (then, Focused:bool=~) {is=title K2Node_FunctionEntry}
   > n2 = Branch(Condition=@n1)
   > PlayAnimationForward(InAnimation=Get Anim_OnFocus() -> (Anim_OnFocus:Widget Animation Object Reference), PlaybackSpeed=10)
   PlayAnimationReverse(execute=@n2.else, InAnimation=Get Anim_OnFocus(), PlaybackSpeed=10)
  ]}
 SetSelected {nodes [
   n1 = SetSelected!() -> (then, IsSelected:bool=~) {is=title K2Node_FunctionEntry}
   > n2 = Branch(Condition=@n1)
   > PlayAnimationForward(InAnimation=Get Anim_Selected() -> (Anim_Selected:Widget Animation Object Reference), PlaybackSpeed=10)
   PlayAnimationReverse(execute=@n2.else, InAnimation=Get Anim_Selected(), PlaybackSpeed=10)
  ]}
]
variables [
 bBackgroundOnlyForNamedSlot {type=boolean}
 Block_Icon {type=WBP_MatchmakingBlock_Image_C, default=null}
 ImagePadding {type=Margin, default=<ltrb 0 0 4 0>}
 ImageSize {type=Vector2D, default=<0 0>}
]
widgets {
 info {inheritedWidgetCount=0, parentClass=&UserWidget, rootWidgetClass=&GridPanel, widgetCount=4}
 widgets [
  Main {widgetClassPath=&GridPanel, bIsVariable=true, widget=Main, properties {bOverride_Cursor=true, columnFill [0,1], rowFill [1]}}
  TextBox {
   widgetClassPath=&VerticalBox
   parent=Main
   widget=TextBox
   slot {class=GridSlot, properties {column=1, verticalAlignment=VAlign_Center}}
  }
  SecondText {
   widgetClassPath=&CommonTextBlock
   bIsVariable=true
   parent=TextBox
   widget=SecondText
   properties {
    bAutoCollapseWithEmptyText=true
    mobileFontSizeMultiplier=1.2
    scrollStyle=&TextScrollStyle-Base_C
    colorAndOpacity {colorUseRule=UseColor_Specified, specifiedColor=<rgba 0.6939 0.6939 0.6939 0.749>}
    font {
     bForceMonospaced=false
     bMaterialIsStencil=false
     fontMaterial=null
     fontObject=&HeadingNow
     letterSpacing=0
     monospacedWidth=1
     size=15
     skewAmount=0
     typefaceFontName=75 Medium
     outlineSettings {
      bApplyOutlineToDropShadows=false
      bMiteredCorners=false
      bSeparateFillAlpha=false
      outlineMaterial=null
      outlineSize=0
      outlineColor=<rgba 0 0 0 1>
     }
    }
    shadowOffset=<0 0>
   }
   slot {class=VerticalBoxSlot}
  }
  Text {
   widgetClassPath=&CommonTextBlock
   bIsVariable=true
   parent=TextBox
   widget=Text
   properties {
    bAutoCollapseWithEmptyText=true
    mobileFontSizeMultiplier=1.2
    scrollStyle=&TextScrollStyle-Base_C
    text=$SettingName
    textTransformPolicy=ToUpper
    colorAndOpacity {colorUseRule=UseColor_Specified, specifiedColor=<rgba 0.8879 0.8879 0.8879 1>}
    font {
     bForceMonospaced=false
     bMaterialIsStencil=false
     fontMaterial=/MatchmakingSettingsFN1Shooters/MatchmakingSettings/MaterialInstances/MI_UI_FN1_Icon
     fontObject=&HeadingNow
     letterSpacing=0
     monospacedWidth=1
     size=15
     skewAmount=0
     typefaceFontName=86 Bold
     outlineSettings {
      bApplyOutlineToDropShadows=false
      bMiteredCorners=false
      bSeparateFillAlpha=false
      outlineMaterial=null
      outlineSize=0
      outlineColor=<rgba 0 0 0 1>
     }
    }
    shadowOffset=<0 0>
   }
   slot {class=VerticalBoxSlot, properties {verticalAlignment=VAlign_Center}}
  }
 ]
}
