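# Base modular button for matchmaking settings rows, with entry info, optional icon and a named slot for the setting control.
# PreConstruct applies compact-UI sizing, layout and background mode, and registers UIKit blocks.
# Hover, selection and enable events drive block transitions and entry-info state, with an instant-selection option.
# The optional description follows hover on non-touch input and focus on touch; focus also drives a conditional outline.
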WidgetBlueprint /MatchmakingSettingsFN1Shooters/MatchmakingSettings/Entries/Buttons/WBP_FN1_MatchmakingSettings_EntryBase
graphs [
 AddDescriptionText {nodes [
   AddDescriptionText() {is=title K2Node_FunctionEntry}
   > n1 = Construct Object from Class(Class=/MatchmakingSettingsFN1Shooters/MatchmakingSettings/WBP_FN1_Matchmaking_EntryDescription.WBP_FN1_Matchmaking_EntryDescription_C) -> (ReturnValue:WBP FN1 Matchmaking Entry Description Object Reference)
   > n2 = Set Description(Description:WBP FN1 Matchmaking Entry Description Object Reference=@n1) -> (Output_Get:WBP FN1 Matchmaking Entry Description Object Reference)
   > n3 = AddChildToGrid!(
    execute
    self:Grid Panel Object Reference=Get Root() -> (Root:Grid Panel Object Reference)
    Content:Widget Object Reference=@n2
    InRow:int=1
    InColumn:int=0
   ) -> (then, ReturnValue:Grid Slot Object Reference)
   > SetHorizontalAlignment(self:Grid Slot Object Reference=@n3, InHorizontalAlignment=HAlign_Center)
   > SetPadding!(
    execute
    self:Grid Slot Object Reference=@n3
    InPadding_Left:float=0
    InPadding_Top:float=24
    InPadding_Right:float=0
    InPadding_Bottom:float=0
   ) -> (then)
   > SetVisibility(self=Get Description() -> (Description:WBP FN1 Matchmaking Entry Description Object Reference), InVisibility=Hidden)
  ]}
 AddIcon {nodes [
   AddIcon() {is=title K2Node_FunctionEntry}
   > Branch(Condition=Get ShowIcon() -> (ShowIcon:bool))
   n1 = Get MatchmakingSettingEntryInfo() -> (MatchmakingSettingEntryInfo:WBP FN1 Matchmaking Settings Setting Entry Info Object Reference)
   > Set ImagePadding!(
    execute
    ImagePadding:Margin Structure=Get Image Padding() -> (Image Padding:Margin Structure)
    self:WBP FN1 Matchmaking Settings Setting Entry Info Object Reference=@n1
   ) -> (then, Output_Get:Margin Structure)
   > Set ImageSize!(
    execute
    ImageSize:Vector 2D Structure=Get Image Size() -> (Image Size:Vector 2D Structure)
    self:WBP FN1 Matchmaking Settings Setting Entry Info Object Reference=@n1
   ) -> (then, Output_Get:Vector 2D Structure)
   > n2 = AddImageBlock!(execute, self:WBP FN1 Matchmaking Settings Setting Entry Info Object Reference=Get MatchmakingSettingEntryInfo()) -> (then, Block_Icon:WBP Matchmaking Block Image Object Reference)
   > Set Block Icon(Block Icon:WBP Matchmaking Block Image Object Reference=@n2) -> (Output_Get:WBP Matchmaking Block Image Object Reference)
   > RegisterModularBlock!(execute, self:UIKit Modular Button Object Reference, Block:UIKit Block Object Reference=@n2) -> (then)
  ]}
 EventGraph {nodes [
   n1 = SetBlocksEnabled!() -> (OutputDelegate, then, Enabled:bool=~) {is=title K2Node_CustomEvent}
   > RequestInstantTransitionOnBlocks!(
    execute
    self:UIKit Modular Button Object Reference
    State:EUIKitBlockInstantTransitionState Enum=Select(
     Option 0:EUIKitBlockInstantTransitionState Enum=InstantDisabled
     Option 1:EUIKitBlockInstantTransitionState Enum=InstantEnabled
     Index=@n1.Enabled
    ) -> (ReturnValue:EUIKitBlockInstantTransitionState Enum)
   ) -> (then)
   > SetEnabled!(execute, self:WBP FN1 Matchmaking Settings Setting Entry Info Object Reference=Get MatchmakingSettingEntryInfo(), Enabled:bool=@n1.Enabled) -> (then)
   n2 = Event PreConstruct()
   > Parent: PreConstruct(IsDesignTime=@n2.IsDesignTime)
   > SetMobileStyle() {is=title K2Node_CallFunction}
   > n3 = Cast To SoundBase!(execute, Object=Get!() -> (PressedSlateSoundOverride_ResourceObject:Object Reference)) -> (then, CastFailed, AsSound Base:Sound Base Object Reference) {is=title K2Node_DynamicCast}
   > n4 = Set PressedSound(PressedSound:Sound Base Object Reference=@n3) -> (Output_Get:Sound Base Object Reference)
   n5 = Get BackgroundBlock() -> (BackgroundBlock:WBP Matchmaking Settings Background Object Reference)
   SetIsSelectable!(
    execute=[@n4.then, @n3.CastFailed]
    self:WBP UIKit Block Base Object Reference=@n5
    Is Selectable:bool=Get bBackgroundSelectable() -> (bBackgroundSelectable:bool)
   ) -> (then)
   > RegisterModularBlock(Block=@n5)
   > SetBlocksTiming!(
    execute
    self:UIKit Modular Button Object Reference
    Timing:"UIKit Block Timing Structure (by ref)"=Get BlockTiming() -> (BlockTiming:UIKit Block Timing Structure)
   ) -> (then)
   > AddIcon() {is=title K2Node_CallFunction}
   n6 = Get NameSlot_SettingType() -> (NameSlot_SettingType:Named Slot Object Reference)
   > SetGridSlotPadding!(execute, Widget:Widget Object Reference=@n6, Padding:Margin Structure=Get NameSlotPaddings() -> (NameSlotPaddings:Margin Structure)) -> (then)
   > SetVisibility(
    self=@n6
    InVisibility=Select(
     Option 0:ESlateVisibility Enum=Collapsed
     Option 1:ESlateVisibility Enum=SelfHitTestInvisible
     Index=Get bNameSlotVisible() -> (bNameSlotVisible:bool)
    ) -> (ReturnValue:ESlateVisibility Enum)
   )
   > SetGridSlotPadding(Widget=Get MatchmakingSettingEntryInfo(), Padding=Get TextPaddings() -> (TextPaddings:Margin Structure))
   > SetGridSlotAlignment!(
    execute
    Widget:Widget Object Reference=Get MatchmakingSettingEntryInfo()
    Horizontal:EHorizontalAlignment Enum=Get Text Horizontal Alignment() -> (Text Horizontal Alignment:EHorizontalAlignment Enum=HAlign_Fill)
    Vertical:EVerticalAlignment Enum=VAlign_Center
   ) -> (then)
   n7 = Get bBackgroundOnlyForNamedSlot() -> (bBackgroundOnlyForNamedSlot:bool)
   > Set bBackgroundOnlyForNamedSlot!(
    execute
    bBackgroundOnlyForNamedSlot:bool=@n7
    self:WBP FN1 Matchmaking Settings Setting Entry Info Object Reference=Get MatchmakingSettingEntryInfo()
   ) -> (then, Output_Get:bool)
   > SetVisibility(
    self=Get Background() -> (Background:Image Object Reference)
    InVisibility=Select(Option 0=Collapsed, Option 1=SelfHitTestInvisible, Index=@n7)
   )
   > Branch(Condition=@n7)
   > AddChildToGrid(self=Get Content() -> (Content:Grid Panel Object Reference), Content=Get BackgroundBlock(), InRow=0, InColumn=1)
   Event OnAddedToFocusPath!() -> (OutputDelegate, then, InFocusEvent:Focus Event Structure) {is=title K2Node_Event}
   > ShowOutline!(
    execute
    self:WBP Matchmaking Settings Outline Object Reference=Get FocusOutline() -> (FocusOutline:WBP Matchmaking Settings Outline Object Reference)
    AdditionalCondition:bool=AND Boolean(A=Get ShouldShowOutline() -> (ShouldShowOutline:bool), B=Get bSelectable() -> (bSelectable:bool))
   ) -> (then)
   n8 = IsTouch!() -> (ReturnValue:bool)
   > Branch(Condition=@n8)
   n9 = Get Description()
   > SetVisibility(self=@n9, InVisibility=SelfHitTestInvisible)
   Event On Deselected()
   > Branch(Condition=Get bSelectable())
   > SetBlocksIsSelected!(execute, self, IsSelected:bool=false) -> (then)
   > RequestTransitionOnBlocks!(
    execute
    self:UIKit Modular Button Object Reference
    State:EUIKitBlockVisualState Enum=Select(Option 0:EUIKitBlockVisualState Enum=DeselectedIdle, Option 1:EUIKitBlockVisualState Enum=DeselectedFocused, Index=IsHovered()) -> (ReturnValue:EUIKitBlockVisualState Enum)
   ) -> (then)
   Event On Hovered()
   > n10 = Branch(Condition=IsTouch())
   RequestTransitionOnBlocks(execute=@n10.else, State=Focused)
   n11 = Get Description()
   > SetVisibility(self=@n11, InVisibility=SelfHitTestInvisible)
   n12 = NOT Boolean(A=Get bBackgroundOnlyForNamedSlot())
   > Branch(Condition=@n12)
   n13 = Get MatchmakingSettingEntryInfo()
   > SetFocused!(execute, self:WBP FN1 Matchmaking Settings Setting Entry Info Object Reference=@n13, Focused:bool=true) -> (then)
   Event On Unhovered()
   > n14 = Branch(Condition=IsTouch())
   RequestTransitionOnBlocks(execute=@n14.else, State=Unfocused)
   > SetVisibility(self=@n11, InVisibility=Hidden)
   > Branch(Condition=@n12)
   > SetFocused(self=@n13, Focused=false)
   Event OnRemovedFromFocusPath!() -> (OutputDelegate, then, InFocusEvent:Focus Event Structure) {is=title K2Node_Event}
   > HideOutline(self:WBP Matchmaking Settings Outline Object Reference=Get FocusOutline())
   > Branch(Condition=@n8)
   > SetVisibility(self=@n9, InVisibility=Hidden)
   Event On Selected()
   > Branch(Condition=Get bSelectable())
   > SetBlocksIsSelected(IsSelected=true)
   > n15 = Branch(Condition=Get Is Instant Selection() -> (Is Instant Selection:bool))
   > RequestInstantTransitionOnBlocks(State=InstantSelected)
   > Set Is Instant Selection(Is Instant Selection:bool=false) -> (Output_Get:bool)
   RequestTransitionOnBlocks(execute=@n15.else, State=Selected)
  ]}
 GetAnalyticsAttribute {nodes [
   GetAnalyticsAttribute() {is=title K2Node_FunctionEntry}
   > Return Node!(execute, Attribute:Analytics Event Attr Structure=Get AnalyticsAttribute() -> (AnalyticsAttribute:Analytics Event Attr Structure))
  ]}
 GetBlockTimings {nodes [
   GetBlockTimings() {is=title K2Node_FunctionEntry}
   > Return Node!(execute, BlockTiming:UIKit Block Timing Structure=Get BlockTiming())
  ]}
 IsTouch {nodes [
   IsTouch() {is=title K2Node_FunctionEntry}
   n1 = Get CommonInputSubsystem()
   > Return Node!(execute, ReturnValue:bool="Equal (Enum)"(A=GetCurrentInputType(self=@n1), B=Touch))
  ]}
 SetBlocksIsSelected {nodes [
   n1 = SetBlocksIsSelected!() -> (then, IsSelected:bool=~) {is=title K2Node_FunctionEntry}
   > n2 = For Each Loop(Array:Array of UIKit Block Object References=Get ModularBlocks() -> (ModularBlocks:Array of UIKit Block Object References)) -> (Array Element:UIKit Block Object Reference)
   n3 = Branch(execute=@n2.Completed, Condition=Get bBackgroundOnlyForNamedSlot())
   SetSelected!(
    execute=@n3.else
    self:WBP FN1 Matchmaking Settings Setting Entry Info Object Reference=Get MatchmakingSettingEntryInfo()
    IsSelected:bool=@n1
   ) -> (then)
   n4 = Cast To WBP_UIKit_Block_Base!(execute=@n2.LoopBody, Object=@n2) -> (then, CastFailed, AsWBP UIKit Block Base:WBP UIKit Block Base Object Reference) {is=title K2Node_DynamicCast}
   > Branch(Condition=GetIsSelectable!(self:WBP UIKit Block Base Object Reference=@n4) -> (IsSelectable:bool))
   > SetIsSelected!(execute, self:WBP UIKit Block Base Object Reference=@n4, IsSelected:bool=@n1) -> (then)
  ]}
 SetDisableState {nodes [
   n1 = SetDisableState!() -> (then, IsDisabled:bool=~) {is=title K2Node_FunctionEntry}
   > SetScalarParameterValue(
    self=Get DynamicBackground() -> (DynamicBackground:Material Instance Dynamic Object Reference)
    ParameterName=IsDisabled
    Value=Select(Option 0:float=0, Option 1:float=1, Index=@n1) -> (ReturnValue:float)
   )
   n2 = Select(Option 0=1, Option 1=0.6, Index=@n1)
   > Set RenderOpacity!(execute, RenderOpacity:float=@n2, self:Widget Object Reference=Get MatchmakingSettingEntryInfo()) -> (then, Output_Get:float)
   > Set RenderOpacity(RenderOpacity=@n2, self=Get NameSlot_SettingType())
  ]}
 SetEnabled {nodes [
   n1 = SetEnabled!() -> (then, Enabled:bool=~) {is=title K2Node_FunctionEntry}
   > SetIsLocked!(execute, self:Common Button Base Object Reference, bInIsLocked:bool=NOT Boolean(A=@n1)) -> (then)
   > SetBlocksEnabled!(execute, self, Enabled:bool=@n1) -> (then) {is=title K2Node_CallFunction}
   > SetPressedSoundOverride!(
    execute
    self:Common Button Base Object Reference
    Sound:Sound Base Object Reference=Select(
     Option 0:Sound Base Object Reference
     Option 1:Sound Base Object Reference=Get PressedSound() -> (PressedSound:Sound Base Object Reference)
     Index=@n1
    ) -> (ReturnValue:Sound Base Object Reference)
   ) -> (then)
   > n2 = For Each Loop(Array=Get ModularBlocks())
   > SetIsEnabled(self=@n2, bInIsEnabled=@n1)
  ]}
 SetIcon {nodes [
   SetMaterial!(
    execute
    self:WBP Matchmaking Block Image Object Reference=Get Block_Icon!(self:WBP FN1 Matchmaking Settings Setting Entry Info Object Reference=Get MatchmakingSettingEntryInfo()) -> (Block_Icon:WBP Matchmaking Block Image Object Reference)
    InMaterial:"Material Instance Object Reference (by ref)"
   ) -> (then)
   SetIcon!() -> (then, InMaterial:"Material Instance Object Reference (by ref)") {is=title K2Node_FunctionEntry}
  ]}
 SetMobileStyle {nodes [
   SetMobileStyle() {is=title K2Node_FunctionEntry}
   > Set Size!(
    execute
    Size:Vector 2D Structure=Select(
     Option 0:Vector 2D Structure=Get MinimumSize() -> (MinimumSize:Vector 2D Structure)
     Option 1:Vector 2D Structure=Get MinimumSize_Mobile() -> (MinimumSize_Mobile:Vector 2D Structure)
     Index=IsUsingCompactUI()
    ) -> (ReturnValue:Vector 2D Structure)
    self:Spacer Object Reference=Get MinSize() -> (MinSize:Spacer Object Reference)
   ) -> (then, Output_Get:Vector 2D Structure)
  ]}
 SetSecondText {nodes [
   n1 = SetSecondText!() -> (then, InText:Text) {is=title K2Node_FunctionEntry}
   > "SetText (Text)"(
    self=Get SecondText!(self:WBP FN1 Matchmaking Settings Setting Entry Info Object Reference=Get MatchmakingSettingEntryInfo()) -> (SecondText:Common Text Object Reference)
    InText=@n1
   )
  ]}
 SetText {nodes [
   n1 = SetText!() -> (then, InText:Text) {is=title K2Node_FunctionEntry}
   > "SetText (Text)"(
    self=Get Text(self:WBP FN1 Matchmaking Settings Setting Entry Info Object Reference=Get MatchmakingSettingEntryInfo()) -> (Text:Common Text Object Reference)
    InText=@n1
   )
  ]}
]
variables [
 AnalyticsAttribute {type=AnalyticsEventAttr, default {name="", value=""}}
 bBackgroundOnlyForNamedSlot {type=boolean}
 bBackgroundSelectable {type=boolean, default=true}
 Block Icon {type=WBP_MatchmakingBlock_Image_C, default=null}
 BlockTiming {
  type=UIKitBlockTiming
  default {
   deselectingFocused=0.15
   deselectingIdle=0.15
   disabling=0
   enabling=0
   focusing=0.15
   hovering=0.15
   pressing=0
   releasing=0
   selecting=0.15
   unfocusing=0.15
   unhovering=0.15
  }
 }
 bNameSlotVisible {type=boolean, default=true}
 Description {type=WBP_FN1_Matchmaking_EntryDescription_C, default=null}
 DescriptionFont {
  type=SlateFontInfo
  default {
   bForceMonospaced=false
   bMaterialIsStencil=false
   fontMaterial=null
   fontObject=&HeadingNow
   letterSpacing=0
   monospacedWidth=1
   size=12.75
   skewAmount=0
   typefaceFontName=75 Medium
   outlineSettings {
    bApplyOutlineToDropShadows=false
    bMiteredCorners=false
    bSeparateFillAlpha=false
    outlineMaterial=null
    outlineSize=0
    outlineColor=<rgba 0 0 0 1>
   }
  }
 }
 DynamicBackground {type=MaterialInstanceDynamic}
 Image Padding {type=Margin, default=<ltrb 0 0 4 0>}
 Image Size {type=Vector2D, default=<32 32>}
 Is Instant Selection {type=boolean}
 MinimumSize {type=Vector2D, default=<600 64>}
 MinimumSize_Mobile {type=Vector2D, default=<600 80>}
 NameSlotPaddings {type=Margin, default=<ltrb 0 0 0 0>}
 PressedSound {type=SoundBase, default=null}
 ShouldShowOutline {type=boolean}
 ShowIcon {type=boolean}
 Text Horizontal Alignment {type=EHorizontalAlignment, default=HAlign_Fill}
 TextPaddings {type=Margin, default=<ltrb 0 8 32 8>}
]
widgets {
 info {inheritedWidgetCount=1, parentClass=/Script/FortniteUI.UIKitModularButton, rootWidgetClass=&GridPanel, widgetCount=9}
 widgets [
  Root {widgetClassPath=&GridPanel, bIsVariable=true, widget=Root, properties {columnFill [1], rowFill [1]}}
  MinSize {widgetClassPath=&Spacer, bIsVariable=true, parent=Root, widget=MinSize, properties {size=<600 64>}, slot {class=GridSlot}}
  Background {
   widgetClassPath=&Image
   bIsVariable=true
   parent=Root
   widget=Background
   properties {colorAndOpacity=<rgba 0.0137 0.0015 0.0513 0.2>}
   slot {class=GridSlot, properties {padding=<ltrb -16 -16 -16 -16>}}
  }
  BackgroundBlock {
   widgetClassPath=/MatchmakingSettingsFN1Shooters/MatchmakingSettings/WBP_MatchmakingSettingsBackground
   bIsVariable=true
   parent=Root
   widget=BackgroundBlock
   slot {class=GridSlot}
  }
  Content {
   widgetClassPath=&GridPanel
   bIsVariable=true
   parent=Root
   widget=Content
   properties {columnFill [1], rowFill [1]}
   slot {class=GridSlot, properties {layer=1}}
  }
  MatchmakingSettingEntryInfo {
   widgetClassPath=&/WBP_FN1_MatchmakingSettings_SettingEntryInfo
   bIsVariable=true
   parent=Content
   widget=MatchmakingSettingEntryInfo
   properties {bOverride_Cursor=true}
   slot {class=GridSlot, properties {verticalAlignment=VAlign_Center, padding=<ltrb 0 8 32 8>}}
  }
  NameSlot_SettingType {
   widgetClassPath=&NamedSlot
   bIsVariable=true
   parent=Content
   widget=NameSlot_SettingType
   slot {class=GridSlot, properties {column=1, layer=5, verticalAlignment=VAlign_Center, padding=<ltrb 60 16 60 16>}}
  }
  FocusOutline {
   widgetClassPath=/MatchmakingSettingsFN1Shooters/MatchmakingSettings/WBP_Matchmaking_Settings_Outline
   bIsVariable=true
   parent=Root
   widget=FocusOutline
   slot {class=GridSlot, properties {layer=1, padding=<ltrb -28 -24 -28 -24>}}
  }
  InputActionWidget {widgetClassPath=&CommonActionWidget, bInherited=true, widget=null}
 ]
}
