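# M_UI_FN1_RankIcon is a translucent UI material for rank icons in the FN1 Shooters matchmaking settings, with an adjustable greyscale effect.
# A Texture parameter supplies the icon image, so material instances can swap in different icons.
# A linear interpolation blends the sampled RGB with a desaturated copy, driven by the Desaturate scalar parameter.
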Material /MatchmakingSettingsFN1Shooters/MatchmakingSettings/Materials/M_UI_FN1_RankIcon
graph {
 properties {bAutomaticallySetUsageInEditor=false, blendMode=BLEND_Translucent, materialDomain=MD_UI, usedShadingModels=Default Lit}
 nodes [
  n1 = TextureSampleParameter2D() {parameterName=Texture}
  LinearInterpolate(A=@n1.RGB, B=Desaturation(None=@n1.RGB), Alpha=ScalarParameter() {parameterName=Desaturate})
 ]
}
