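# Ranked matchmaking settings entry widget: a toggle beside the player's current rank name and icon.
# Clicking swaps the ranked setting view model between its two options, fires analytics and plays a sound; the value handler updates toggle, option text and description.
# Ranked view model changes look up tier display data with overrides, then async-load the icon, discarding stale loads.
# The icon material desaturates and rank visuals dim when ranked is off.
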WidgetBlueprint /MatchmakingSettingsFN1Shooters/MatchmakingSettings/Entries/Buttons/WBP_FN1_MatchmakingSettings_Ranked
graphs [
 EventGraph {nodes [
   n1 = UpdateRankIcon!() -> (OutputDelegate, then, Texture:Texture 2D Soft Object Reference) {is=title K2Node_CustomEvent}
   > n2 = Set CurrentRankIcon(CurrentRankIcon:Texture 2D Soft Object Reference=@n1.Texture) -> (Output_Get:Texture 2D Soft Object Reference)
   > n3 = Load From Soft Texture!(execute:Exec, Texture:Texture 2D Soft Object Reference=@n2) -> (then:Exec, Completed:Exec, Result:Texture 2D Object Reference) {is=title MVVMK2Node_LoadSoftTexture}
   n4 = Get CurrentRankIcon() -> (CurrentRankIcon:Texture 2D Soft Object Reference)
   n5 = To Soft Object Reference(SoftObjectPath=GetSoftObjectPath(Object=@n3))
   Branch(
    execute=@n3.Completed
    Condition="Equal Exactly (String)"(
     A="To String (SoftObjectReference)"!(SoftObjectReference:"Soft Object Reference (by ref)"=@n5) -> (ReturnValue:String)
     B="To String (SoftObjectReference)"(SoftObjectReference=@n4)
    )
   )
   > SetTextureParameterValue(self=Get MID_RankIcon() -> (MID_RankIcon:Material Instance Dynamic Object Reference), ParameterName=Texture, Value=@n3)
   > UpdateRankedState!(execute, self, InToggleOn:bool=Get IsToggleOn() -> (IsToggleOn:bool)) -> (then)
   > SetVisibility(
    self=[Get Txt_Rank() -> (Txt_Rank:Common Text Object Reference), Get RankIcon() -> (RankIcon:Image Object Reference)]
    InVisibility=SelfHitTestInvisible
   )
   n6 = Event OnListItemObjectSet()
   > n7 = Cast To FortMatchmakingSettingRankedVM!(execute, Object=@n6.ListItemObject) -> (then, CastFailed, AsFort Matchmaking Setting Ranked VM:Fort Matchmaking Setting Ranked VM Object Reference) {is=title K2Node_DynamicCast}
   n8 = GetOptions!(self:Fort Matchmaking Setting Discrete Enum VM Object Reference=@n7) -> (ReturnValue:Array of Matchmaking Setting Option Structures)
   > Set False Value(False Value:String=Break Matchmaking Setting Option!(MatchmakingSettingOption:"Matchmaking Setting Option Structure (by ref)"="Get (a ref)"(Array:Array of Matchmaking Setting Option Structures=@n8, Dimension 1=0) -> (Output:"Matchmaking Setting Option Structure (by ref)")) -> (bIsEnabled:bool=~, Value:String, Name:Text) {is=title K2Node_BreakStruct}.Value) -> (Output_Get:String)
   > Set True Value(True Value:String=Break Matchmaking Setting Option(MatchmakingSettingOption="Get (a ref)"(Array=@n8, Dimension 1=1)) -> (bIsEnabled=~).Value) -> (Output_Get:String)
   > Set FortMatchmakingSettingRankedVM(FortMatchmakingSettingRankedVM:Fort Matchmaking Setting Ranked VM Object Reference=@n7) -> (Output_Get:Fort Matchmaking Setting Ranked VM Object Reference)
   Event On Clicked()
   n9 = Get FortMatchmakingSettingRankedVM() -> (FortMatchmakingSettingRankedVM:Fort Matchmaking Setting Ranked VM Object Reference)
   > Is Valid(InputObject=@n9)
   n10 = "Equal Exactly (String)"(
    A=GetValue!(self:Fort Matchmaking Setting Discrete Enum VM Object Reference=@n9) -> (ReturnValue:String)
    B=Get False Value() -> (False Value:String)
   )
   > SetValue!(
    execute
    self:Fort Matchmaking Setting Discrete Enum VM Object Reference=@n9
    Value:String=SelectString(A=Get True Value() -> (True Value:String), B=Get False Value(), bPickA=@n10)
    LocalPlayer:Local Player Object Reference=GetOwningLocalPlayer!(self:Widget Object Reference) -> (ReturnValue:Local Player Object Reference)
   ) -> (then)
   n11 = Get!() -> (AnalyticsAttribute_Name:String, AnalyticsAttribute_Value:String)
   > Fire Event Element Interacted(
    Element=Self-Reference()
    AdditionalInfo="Add String (Non-Localized)"!(
     Info:"Flow Analytics Additional Info Structure (by ref)"=Make Flow Analytics Additional Info!() -> (FlowAnalyticsAdditionalInfo:Flow Analytics Additional Info Structure) {is=title K2Node_MakeStruct}
     Key:String=@n11
     Value:String=@n11.AnalyticsAttribute_Value
    ) -> (ReturnValue:Flow Analytics Additional Info Structure)
   )
   > PlaySound2D(Sound=Select(
     Option 0:Sound Base Object Reference=Get False_Sound() -> (False_Sound:Sound Base Object Reference)
     Option 1:Sound Base Object Reference=Get True_Sound() -> (True_Sound:Sound Base Object Reference)
     Index=@n10
    ) -> (ReturnValue:Sound Base Object Reference))
   Event Construct()
   > InitializeRankedViewModal(self:Ranked View Model Object Reference=Get FortRankedViewModel() -> (FortRankedViewModel:Ranked View Model Object Reference))
   n12 = Event PreConstruct()
   > Parent: PreConstruct(IsDesignTime=@n12.IsDesignTime)
   > AddChildToGrid!(
    execute
    self:Grid Panel Object Reference=Get BG() -> (BG:Grid Panel Object Reference)
    Content:Widget Object Reference=Get BackgroundBlock() -> (BackgroundBlock:WBP Matchmaking Settings Background Object Reference)
    InRow:int=0
    InColumn:int=0
   ) -> (then, ReturnValue:Grid Slot Object Reference)
   > SetTextColumFillRule!(execute, self, Ranked:bool=true) -> (then)
   > AddDescriptionText(self:WBP FN1 Matchmaking Settings Entry Base Object Reference)
   > n13 = GetDynamicMaterial(self=Get RankIcon())
   > Set MID_RankIcon(MID_RankIcon:Material Instance Dynamic Object Reference=@n13) -> (Output_Get:Material Instance Dynamic Object Reference)
  ]}
 HandleCurrentRankChanged {nodes [
   HandleCurrentRankChanged!() -> (then, CurrentRank:int) {is=title K2Node_FunctionEntry}
   > SetCurrentPlayerRank()
  ]}
 HandleRankTypeChanged {nodes [
   HandleRankTypeChanged!() -> (then, RankType:String) {is=title K2Node_FunctionEntry}
   > SetCurrentPlayerRank()
  ]}
 HandleSettingValueChanged {nodes [
   n1 = HandleSettingValueChanged!() -> (then, Value:String) {is=title K2Node_FunctionEntry}
   n2 = "Equal Exactly (String)"(A=@n1, B=Get True Value())
   > Set IsToggleOn(IsToggleOn:bool=@n2) -> (Output_Get:bool)
   > SetToggleState!(execute, self, ToggleOn:bool=@n2) -> (then)
  ]}
 IsEnabled {nodes [
   n1 = IsEnabled!() -> (then, Enable:bool=~) {is=title K2Node_FunctionEntry}
   > SetIsEnabled(bInIsEnabled=@n1)
   > SetEnabled!(execute, self:WBP FN1 Toggle Object Reference=Get FN1_Toggle() -> (FN1_Toggle:WBP FN1 Toggle Object Reference), Enabled:bool=@n1) -> (then)
  ]}
 SetCurrentPlayerRank {nodes [
   SetCurrentPlayerRank() {is=title K2Node_FunctionEntry}
   > SetVisibility(self=[Get Txt_Rank(), Get RankIcon()], InVisibility=Hidden)
   n1 = Get FortRankedViewModel()
   n2 = Get bIsPlayerRanked!(self:Ranked View Model Object Reference=@n1) -> (bIsPlayerRanked:bool)
   > SetTextColumFillRule(Ranked=@n2)
   > Branch(Condition=@n2)
   n3 = Get RankType!(self:Ranked View Model Object Reference=@n1) -> (RankType:String)
   n4 = GetHabaneroDisplayDataOverride!(
    InDisplayData:Fort Habanero Display Data Object Reference=Get RankDisplayData() -> (RankDisplayData:Fort Habanero Display Data Object Reference)
    InOverrideData:Fort Compete UIHabanero Tier Override List Object Reference=Get RankInfoOverrides() -> (RankInfoOverrides:Fort Compete UIHabanero Tier Override List Object Reference)
    InKey:"Fort Compete UIHabanero Tier Override Key Structure (by ref)"=Make Fort Compete UIHabanero Tier Override Key!(
     RankedTrack:Name=String To Name(InString=@n3)
     Season:String
     Rank:int=Get CurrentRank!(self:Ranked View Model Object Reference=@n1) -> (CurrentRank:int=0)
     MatchmakingRating:int=Get MatchmakingRating!(self:Ranked View Model Object Reference=@n1) -> (MatchmakingRating:int=0)
     CurrentPosition:int=Get CurrentPlayerPosition!(self:Ranked View Model Object Reference=@n1) -> (CurrentPlayerPosition:int=0)
    ) -> (FortCompeteUIHabaneroTierOverrideKey:Fort Compete UIHabanero Tier Override Key Structure) {is=title K2Node_MakeStruct}
   ) -> (OutData:Fort Habanero Tier Override Structure, ReturnValue:bool)
   > Branch(Condition=@n4.ReturnValue)
   n5 = Break Fort Habanero Tier!(FortHabaneroTier:"Fort Habanero Tier Structure (by ref)"=@n4) -> (
    ForegroundIcon:Material Instance Soft Object Reference
    ForegroundIconSmall:Material Instance Soft Object Reference
    BackgroundIcon:Material Instance Soft Object Reference
    BackgroundIconSmall:Material Instance Soft Object Reference
    WhiteIcon:Texture 2D Soft Object Reference
    SimpleIconSmall:Texture 2D Soft Object Reference
    Name:Text
    TierType:Fort Habanero Tier Type Data Object Reference
    AdditionalLeftIcon:Material Instance Soft Object Reference
    AdditionalRightIcon:Material Instance Soft Object Reference
   ) {is=title K2Node_BreakStruct}
   > "SetText (Text)"(self=Get Txt_Rank(), InText=@n5.Name)
   > UpdateRankIcon!(execute, self, Texture:Texture 2D Soft Object Reference=@n5.SimpleIconSmall) -> (then) {is=title K2Node_CallFunction}
  ]}
 SetEnabled {nodes [
   n1 = SetEnabled!() -> (then, Enabled:bool) {is=title K2Node_FunctionEntry}
   > Parent: SetEnabled!(execute, Enabled:bool=@n1) -> (then) {is=title K2Node_CallParentFunction}
   > SetEnabled(self=Get FN1_Toggle(), Enabled=@n1) {is=title K2Node_CallFunction}
   > UpdateDescription()
   > UpdateIsLockedText!(execute, self, Enabled:bool=Get Enabled() -> (Enabled:bool)) -> (then)
  ]}
 SetTextColumFillRule {nodes [
   n1 = SetTextColumFillRule!() -> (then, Ranked:bool=~) {is=title K2Node_FunctionEntry}
   n2 = Get Content() -> (Content:Grid Panel Object Reference)
   > SetColumnFill!(
    execute
    self:Grid Panel Object Reference=@n2
    ColumnIndex:int=0
    Coefficient:float=Select(Option 0:float=1, Option 1:float=0.3, Index=@n1) -> (ReturnValue:float)
   ) -> (then)
   > SetColumnFill(self=@n2, ColumnIndex=1, Coefficient=Select(Option 0=1, Option 1=0.7, Index=@n1))
  ]}
 SetToggleState {nodes [
   n1 = SetToggleState!() -> (then, ToggleOn:bool=~) {is=title K2Node_FunctionEntry}
   > SetToggle!(execute, self:WBP FN1 Toggle Object Reference=Get FN1_Toggle(), On:bool=@n1) -> (then)
   n2 = "Get (a ref)"!(Array=GetOptions(self=Get FortMatchmakingSettingRankedVM()), Dimension 1="To Integer (Boolean)"!(InBool:bool=@n1) -> (ReturnValue:int=0)) -> (Output_bIsEnabled:bool=~, Output_Value:String, Output_Name:Text)
   > SetText(self:WBP FN1 Matchmaking Settings Entry Base Object Reference, InText:Text=@n2.Output_Name)
   > Set!(
    execute
    AnalyticsAttribute_Name:String=GetKey!(self:Fort Matchmaking Setting Discrete VM Object Reference=Get FortMatchmakingSettingRankedVM()) -> (ReturnValue:String)
    AnalyticsAttribute_Value:String="To String (Name)"(InName=String To Name(InString=@n2.Output_Value))
   ) -> (then, Output_Get:Analytics Event Attr Structure)
   > UpdateRankedState(InToggleOn=Get ToggleOn() -> (ToggleOn:bool))
   > UpdateDescription()
  ]}
 UpdateDescription {nodes [
   UpdateDescription() {is=title K2Node_FunctionEntry}
   n1 = Get FortMatchmakingSettingRankedVM()
   > SetText(
    self:WBP FN1 Matchmaking Entry Description Object Reference=Get Description() -> (Description:WBP FN1 Matchmaking Entry Description Object Reference)
    InText=GetOptionDescription!(self:Fort Matchmaking Setting Discrete Enum VM Object Reference=@n1, Value:String=GetValue(self=@n1)) -> (ReturnValue:Text)
   )
  ]}
 UpdateIsLockedText {nodes [
   n1 = UpdateIsLockedText!() -> (then, Enabled:bool=~) {is=title K2Node_FunctionEntry}
   n2 = GetOptions(self=Get FortMatchmakingSettingRankedVM())
   n3 = "To Integer (Boolean)"(InBool=Get IsToggleOn()) -> (ReturnValue=0)
   > SetText(
    self:WBP FN1 Matchmaking Settings Entry Base Object Reference
    InText=Select(Option 0:Text="LOCTEXT(\"Locked\")", Option 1:Text="Get (a ref)"(Array=@n2, Dimension 1=@n3) -> (Output_bIsEnabled=~).Output_Name, Index=@n1) -> (ReturnValue:Text)
   )
  ]}
 UpdateRankedState {nodes [
   n1 = UpdateRankedState!() -> (then, InToggleOn:bool=~) {is=title K2Node_FunctionEntry}
   > SetRenderOpacity(self=Get Txt_Rank(), InOpacity=Select(Option 0:float=0.75, Option 1:float=1, Index=@n1) -> (ReturnValue:float))
   > SetRenderOpacity(self=Get RankIcon(), InOpacity=Select(Option 0=0.5, Option 1=1, Index=@n1))
   > SetScalarParameterValue(self=Get MID_RankIcon(), ParameterName=Desaturate, Value="To Float (Boolean)"(InBool=NOT Boolean(A=@n1)))
  ]}
]
variables [
 CurrentRankIcon {type=Texture2D}
 Description Text {type=CommonTextBlock, default=null}
 False {type=SoundBase, default=null}
 False Value {type=string, default=""}
 False_Sound {type=SoundBase, default=/Game/Sounds/Fort_UI_Sounds/FrontEnd/New_FrontEnd/Ranked/Fortnite_UI_Ranked_On_Big_Cue}
 IsToggleOn {type=boolean}
 MID_RankIcon {type=MaterialInstanceDynamic}
 RankDisplayData {type=FortHabaneroDisplayData, default=/Game/UI/Habanero/HabaneroDisplayData}
 RankInfoOverrides {type=FortCompeteUIHabaneroTierOverrideList, default=/Game/UI/Habanero/DA_RankDisplayDataOverride}
 True {type=SoundBase, default=null}
 True Value {type=string, default=""}
 True_Sound {type=SoundBase, default=/Game/Sounds/Fort_UI_Sounds/FrontEnd/New_FrontEnd/Ranked/Fortnite_UI_Ranked_Off_Big_Cue}
]
widgets {
 info {
  inheritedWidgetCount=0
  namedSlotCount=1
  parentClass=&/WBP_FN1_MatchmakingSettings_EntryBase.WBP_FN1_MatchmakingSettings_EntryBase_C
  rootWidgetClass=null
  widgetCount=5
 }
 widgets [
  HorizontalBox_0 {widgetClassPath=&HorizontalBox, widget=HorizontalBox_0}
  Txt_Rank {
   widgetClassPath=&CommonTextBlock
   bIsVariable=true
   parent=HorizontalBox_0
   widget=Txt_Rank
   properties {
    mobileFontSizeMultiplier=1.2
    renderOpacity=0.75
    scrollStyle=&TextScrollStyle-Base_C
    text=$Rank
    colorAndOpacity {colorUseRule=UseColor_Specified, specifiedColor=<rgba 0.6939 0.6939 0.6939 1>}
    font {
     bForceMonospaced=false
     bMaterialIsStencil=false
     fontMaterial=null
     fontObject=&HeadingNow
     letterSpacing=0
     monospacedWidth=1
     size=15
     skewAmount=0
     typefaceFontName=75 Medium
     outlineSettings {
      bApplyOutlineToDropShadows=false
      bMiteredCorners=false
      bSeparateFillAlpha=false
      outlineMaterial=null
      outlineSize=0
      outlineColor=<rgba 0 0 0 1>
     }
    }
    shadowOffset=<0 0>
   }
   slot {
    class=HorizontalBoxSlot
    properties {horizontalAlignment=HAlign_Right, verticalAlignment=VAlign_Center, padding=<ltrb 0 0 8 0>, size {sizeRule=Fill, value=1}}
   }
  }
  RankIcon {
   widgetClassPath=&Image
   bIsVariable=true
   parent=HorizontalBox_0
   widget=RankIcon
   properties {
    visibility=Hidden
    brush {
     bIsDynamicallyLoaded=false
     drawAs=Image
     imageType=FullColor
     mirroring=NoMirror
     resourceName=None
     resourceObject=/MatchmakingSettingsFN1Shooters/MatchmakingSettings/Materials/M_UI_FN1_RankIcon
     tiling=NoTile
     imageSize=<48 48>
     margin=<ltrb 0 0 0 0>
     outlineSettings {
      bUseBrushTransparency=false
      roundingType=HalfHeightRadius
      width=0
      color {colorUseRule=UseColor_Specified, specifiedColor=<rgba 0 0 0 0>}
      cornerRadii=<0 0 0 0>
     }
     tintColor {colorUseRule=UseColor_Specified, specifiedColor=<rgba 1 1 1 1>}
     uVRegion {bIsValid=false, max=<0 0>, min=<0 0>}
    }
   }
   slot {class=HorizontalBoxSlot, properties {horizontalAlignment=HAlign_Right, verticalAlignment=VAlign_Center, padding=<ltrb 0 0 16 0>}}
  }
  BG {
   widgetClassPath=&GridPanel
   bIsVariable=true
   parent=HorizontalBox_0
   widget=BG
   properties {columnFill [1], rowFill [1]}
   slot {class=HorizontalBoxSlot}
  }
  FN1_Toggle {
   widgetClassPath=/MatchmakingSettingsFN1Shooters/MatchmakingSettings/Entries/WBP_FN1_Toggle
   bIsVariable=true
   parent=BG
   widget=FN1_Toggle
   slot {class=GridSlot, properties {horizontalAlignment=HAlign_Center, layer=1, verticalAlignment=VAlign_Center, padding=<ltrb 60 16 60 16>}}
  }
 ]
}
