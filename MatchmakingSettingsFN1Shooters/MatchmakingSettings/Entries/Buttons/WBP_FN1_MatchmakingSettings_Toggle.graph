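# Toggle button list entry for a two-option matchmaking setting, derived from the shared matchmaking settings entry base.
# On list item set, it stores the discrete enum view model and maps options 0 and 1 to false and true values.
# Clicking writes the opposite value for the owning player and fires an analytics interaction event.
# The value-changed handler syncs toggle, label, analytics attribute and description, while SetEnabled swaps the label to "Locked" when disabled.
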WidgetBlueprint /MatchmakingSettingsFN1Shooters/MatchmakingSettings/Entries/Buttons/WBP_FN1_MatchmakingSettings_Toggle
graphs [
 EventGraph {nodes [
   n1 = Event OnListItemObjectSet()
   > n2 = Cast To FortMatchmakingSettingDiscreteEnumVM!(execute, Object=@n1.ListItemObject) -> (then, CastFailed, AsFort Matchmaking Setting Discrete Enum VM:Fort Matchmaking Setting Discrete Enum VM Object Reference) {is=title K2Node_DynamicCast}
   n3 = GetOptions!(self:Fort Matchmaking Setting Discrete Enum VM Object Reference=@n2) -> (ReturnValue:Array of Matchmaking Setting Option Structures)
   > Set False Value(False Value:String=Break Matchmaking Setting Option!(MatchmakingSettingOption:"Matchmaking Setting Option Structure (by ref)"="Get (a ref)"(Array:Array of Matchmaking Setting Option Structures=@n3, Dimension 1=0) -> (Output:"Matchmaking Setting Option Structure (by ref)")) -> (bIsEnabled:bool=~, Value:String, Name:Text) {is=title K2Node_BreakStruct}.Value) -> (Output_Get:String)
   > Set True Value(True Value:String=Break Matchmaking Setting Option(MatchmakingSettingOption="Get (a ref)"(Array=@n3, Dimension 1=1)) -> (bIsEnabled=~).Value) -> (Output_Get:String)
   > Set FortMatchmakingSettingDiscreteEnumVM(FortMatchmakingSettingDiscreteEnumVM:Fort Matchmaking Setting Discrete Enum VM Object Reference=@n2) -> (Output_Get:Fort Matchmaking Setting Discrete Enum VM Object Reference)
   Event On Clicked()
   n4 = Get FortMatchmakingSettingDiscreteEnumVM() -> (FortMatchmakingSettingDiscreteEnumVM:Fort Matchmaking Setting Discrete Enum VM Object Reference)
   > Is Valid(InputObject=@n4)
   n5 = GetValue!(self:Fort Matchmaking Setting Discrete Enum VM Object Reference=@n4) -> (ReturnValue:String)
   n6 = Get False Value() -> (False Value:String)
   > SetValue!(
    execute
    self:Fort Matchmaking Setting Discrete Enum VM Object Reference=@n4
    Value:String=SelectString(A=Get True Value() -> (True Value:String), B=Get False Value(), bPickA="Equal Exactly (String)"(A=@n5, B=@n6))
    LocalPlayer:Local Player Object Reference=GetOwningLocalPlayer!(self:Widget Object Reference) -> (ReturnValue:Local Player Object Reference)
   ) -> (then)
   n7 = Get!() -> (AnalyticsAttribute_Name:String, AnalyticsAttribute_Value:String)
   > Fire Event Element Interacted(
    Element=Self-Reference()
    AdditionalInfo="Add String (Non-Localized)"!(
     Info:"Flow Analytics Additional Info Structure (by ref)"=Make Flow Analytics Additional Info!() -> (FlowAnalyticsAdditionalInfo:Flow Analytics Additional Info Structure) {is=title K2Node_MakeStruct}
     Key:String=@n7
     Value:String=@n7.AnalyticsAttribute_Value
    ) -> (ReturnValue:Flow Analytics Additional Info Structure)
   )
   n8 = Event PreConstruct()
   > Parent: PreConstruct(IsDesignTime=@n8.IsDesignTime)
   > AddDescriptionText(self:WBP FN1 Matchmaking Settings Entry Base Object Reference)
  ]}
 HandleSettingValueChanged {nodes [
   n1 = HandleSettingValueChanged!() -> (then, Index:String) {is=title K2Node_FunctionEntry}
   > n2 = Set IsToggleOn(IsToggleOn:bool="Equal Exactly (String)"(A=@n1, B=Get True Value())) -> (Output_Get:bool)
   > SetToggle!(execute, self:WBP FN1 Toggle Object Reference=Get FN1_Toggle() -> (FN1_Toggle:WBP FN1 Toggle Object Reference), On:bool=@n2) -> (then)
   n3 = Get FortMatchmakingSettingDiscreteEnumVM()
   n4 = "Get (a ref)"!(Array=GetOptions(self=@n3), Dimension 1="To Integer (Boolean)"!(InBool:bool=@n2) -> (ReturnValue:int=0)) -> (Output_bIsEnabled:bool=~, Output_Value:String, Output_Name:Text)
   > SetText(self:WBP FN1 Matchmaking Settings Entry Base Object Reference, InText:Text=@n4.Output_Name)
   > Set!(
    execute
    AnalyticsAttribute_Name:String=GetKey!(self:Fort Matchmaking Setting Discrete VM Object Reference=@n3) -> (ReturnValue:String)
    AnalyticsAttribute_Value:String=@n4.Output_Value
   ) -> (then, Output_Get:Analytics Event Attr Structure)
   > UpdateDescription()
  ]}
 SetEnabled {nodes [
   n1 = SetEnabled!() -> (then, Enabled:bool) {is=title K2Node_FunctionEntry}
   > Parent: SetEnabled!(execute, Enabled:bool=@n1) -> (then) {is=title K2Node_CallParentFunction}
   > SetEnabled!(execute, self:WBP FN1 Toggle Object Reference=Get FN1_Toggle(), Enabled:bool=@n1) -> (then) {is=title K2Node_CallFunction}
   > UpdateDescription()
   > UpdateIsLockedText!(execute, self, Enabled:bool=Get Enabled() -> (Enabled:bool)) -> (then)
  ]}
 UpdateDescription {nodes [
   UpdateDescription() {is=title K2Node_FunctionEntry}
   n1 = Get FortMatchmakingSettingDiscreteEnumVM()
   > SetText(
    self:WBP FN1 Matchmaking Entry Description Object Reference=Get Description() -> (Description:WBP FN1 Matchmaking Entry Description Object Reference)
    InText=GetOptionDescription!(self:Fort Matchmaking Setting Discrete Enum VM Object Reference=@n1, Value:String=GetValue(self=@n1)) -> (ReturnValue:Text)
   )
  ]}
 UpdateIsLockedText {nodes [
   n1 = UpdateIsLockedText!() -> (then, Enabled:bool=~) {is=title K2Node_FunctionEntry}
   n2 = GetOptions(self=Get FortMatchmakingSettingDiscreteEnumVM())
   n3 = "To Integer (Boolean)"(InBool=Get IsToggleOn() -> (IsToggleOn:bool)) -> (ReturnValue=0)
   > SetText(
    self:WBP FN1 Matchmaking Settings Entry Base Object Reference
    InText=Select(Option 0:Text="LOCTEXT(\"Locked\")", Option 1:Text="Get (a ref)"(Array=@n2, Dimension 1=@n3) -> (Output_bIsEnabled=~).Output_Name, Index=@n1) -> (ReturnValue:Text)
   )
  ]}
]
variables [
 False Value {type=string, default=""}
 IsToggleOn {type=boolean}
 On {type=boolean, default=true}
 True Value {type=string, default=""}
]
widgets {
 info {
  inheritedWidgetCount=0
  namedSlotCount=1
  parentClass=&/WBP_FN1_MatchmakingSettings_EntryBase.WBP_FN1_MatchmakingSettings_EntryBase_C
  rootWidgetClass=null
  widgetCount=2
 }
 widgets [
  Overlay_1 {widgetClassPath=&Overlay, widget=Overlay_1}
  FN1_Toggle {
   widgetClassPath=/MatchmakingSettingsFN1Shooters/MatchmakingSettings/Entries/WBP_FN1_Toggle
   bIsVariable=true
   parent=Overlay_1
   widget=FN1_Toggle
   slot {class=OverlaySlot, properties {horizontalAlignment=HAlign_Center, verticalAlignment=VAlign_Center, padding=<ltrb 60 16 60 16>}}
  }
 ]
}
